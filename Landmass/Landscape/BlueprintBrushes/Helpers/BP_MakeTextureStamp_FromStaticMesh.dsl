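; Editor helper Blueprint that bakes a static mesh into a height texture stamp for Landmass landscape brushes.
; The construction script centers the mesh on the actor and runs Capture, which applies a height-normalizing material with displacement settings.
; An orthographic scene capture sized from the mesh bounds renders only that mesh, then M_MakeTexture draws the result into an R32f render target.
; The Create Texture event saves that target as a single-float static texture.
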(event EventBeginPlay)

(event Collision|EventActorBeginOverlap
  (params (OtherActor :type Actor%20Object%20Reference)))

(event EventTick
  (params (DeltaSeconds :type Float%20%28single-precision%29)))

(event Custom|Create%20Texture
  (EditorScripting|Rendering|RenderTarget2DCreateStaticTexture2DEditorOnly (Variables|Default|GetOutput%20RT) (Utilities|String|Append (Variables|Default|GetNew%20Asset%20Path) (Variables|Default|GetNew%20Texture%20Name)) "TC_SingleFloat"))

(fn Capture
  (bind (_origin _boxextent _sphereradius) (Collision|GetComponentBounds (Variables|Default|GetStaticMeshComponent)))
  (bind (_origin_2 _boxextent_2 _sphereradius_2) (Collision|GetComponentBounds (Variables|Default|GetStaticMeshComponent)))
  (bind _scenecapturecomponent2d (Variables|Default|GetSceneCaptureComponent2D))
  (bind _capturemid (Variables|Default|GetCaptureMID))
  (bind _outputmid (Variables|Default|GetOutputMID))
  (Rendering|CreateAndAssignMID (Variables|Default|GetCaptureMID) "/Landmass/Landscape/BlueprintBrushes/Helpers/M_NormalizeHeight.M_NormalizeHeight"
    (:Out
      (for _index (range (Rendering|Material|GetNumMaterials (Variables|Default|GetStaticMeshComponent)))
        (Rendering|Material|SetMaterial (Variables|Default|GetStaticMeshComponent) _index (Variables|Default|GetCaptureMID)))
      (Variables|/Script/Engine.SceneCaptureComponent2D|SetOrthoWidth (* (* (+ _sphereradius_2 (Variables|Default|GetZ%20Expansion)) (Variables|Default|GetCapture%20Scale)) 2.0) (Variables|Default|GetSceneCaptureComponent2D))
      (Rendering|SceneCapture|ShowOnlyComponent (Variables|Default|GetSceneCaptureComponent2D) (Variables|Default|GetStaticMeshComponent))
      (Transformation|SetRelativeLocation (Variables|Default|GetSceneCaptureComponent2D) (Math|Vector|MakeVector 0.0 0.0 (* _sphereradius 4.0)) false false)
      (Rendering|Material|SetScalarParameterValue _capturemid "Z Expansion" (Variables|Default|GetZ%20Expansion) :!class /Script/Engine.MaterialInstanceDynamic)
      (Rendering|Material|SetTextureParameterValue _capturemid "Displacement" (Variables|Default|GetDisplacement))
      (Rendering|Material|SetScalarParameterValue _capturemid "Midpoint" (Variables|Default|GetMidpoint) :!class /Script/Engine.MaterialInstanceDynamic)
      (Rendering|Material|SetScalarParameterValue _capturemid "DisplacementHeight" (Variables|Default|GetDisplacementScale) :!class /Script/Engine.MaterialInstanceDynamic)
      (Rendering|CreateAndAssignRenderTarget (Variables|Default|GetCaptureRT) (Variables|Default|GetResolution) (Variables|Default|GetResolution)
        (:Out
          (Rendering|CreateAndAssignRenderTarget (Variables|Default|GetOutput%20RT) (Variables|Default|GetResolution) (Variables|Default|GetResolution) "RTF_R32f"
            (:Out
              (Variables|/Script/Engine.SceneCaptureComponent2D|SetTextureTarget (Variables|Default|GetCaptureRT) _scenecapturecomponent2d)
              (Rendering|SceneCapture|CaptureScene _scenecapturecomponent2d :!class /Script/Engine.SceneCaptureComponent2D)
              (Rendering|CreateAndAssignMID _outputmid "/Landmass/Landscape/BlueprintBrushes/Helpers/M_MakeTexture.M_MakeTexture"
                (:Out
                  (Rendering|Material|SetTextureParameterValue _outputmid "RT" (Variables|Default|GetCaptureRT))
                  (Rendering|DrawMaterialtoRenderTarget (Variables|Default|GetOutput%20RT) (Variables|Default|GetOutputMID)))))))))))

(fn ConstructionScript
  (bind _staticmeshcomponent (Variables|Default|GetStaticMeshComponent))
  (Components|StaticMesh|SetStaticMesh _staticmeshcomponent (Variables|Default|GetStatic%20Mesh))
  (Transformation|SetRelativeLocation _staticmeshcomponent (* (- (Collision|GetComponentBounds (Variables|Default|GetStaticMeshComponent)) (Transformation|GetActorLocation)) -1.0) false false)
  (CallFunction|Capture))
