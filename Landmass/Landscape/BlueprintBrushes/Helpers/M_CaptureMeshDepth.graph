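# Unlit Landmass helper material for landscape blueprint brushes that encodes a mesh's height for depth capture.
# It takes world Z relative to the object's position, divides by the CaptureSize parameter, offsets by 0.5 and saturates, so the object origin sits at mid-range.
# The normalized height is split into a coarse 256-step value and a fractional remainder, packed into two channels for higher precision than one 8-bit channel.
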Material /Landmass/Landscape/BlueprintBrushes/Helpers/M_CaptureMeshDepth
graph {
 properties {bUsedWithNanite=true, pixelDepthOffsetMode=PDOM_Legacy, shadingModel=MSM_Unlit, usedShadingModels=Unlit}
 nodes [
  n1 = WorldPosition()
  n2 = ObjectPositionWS()
  n3 = Divide(
   A=Subtract(A=ComponentMask(None=@n1.XYZ) {b=true}, B=ComponentMask(None=@n2) {b=true})
   B=ScalarParameter() {defaultValue=5000, parameterName=CaptureSize}
  )
  n4 = Multiply(A=Saturate(None=Add(A=@n3) {constB=0.5})) {constB=256}
  AppendVector(A=Divide(A=Floor(None=@n4)) {constB=256}, B=Frac(None=@n4))
 ]
}
