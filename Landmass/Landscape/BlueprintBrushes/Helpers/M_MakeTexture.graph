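# Unlit helper material for Landmass landscape Blueprint brushes that merges two channels of a texture into one scalar value.
# It samples the RT texture parameter as linear color, with a black 1x1 EXR as the placeholder default.
# Red serves as the coarse part and green, divided by 255, is added as the fine part, in the manner of decoding a high/low byte-packed value.
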Material /Landmass/Landscape/BlueprintBrushes/Helpers/M_MakeTexture
graph {
 properties {pixelDepthOffsetMode=PDOM_Legacy, shadingModel=MSM_Unlit, usedShadingModels=Unlit}
 nodes [
  n1 = TextureSampleParameter2D() {
   parameterName=RT
   samplerType=SAMPLERTYPE_LinearColor
   texture=/Engine/Functions/Engine_MaterialFunctions02/PivotPainter2/Black_1x1_EXR_Texture
  }
  Add(A=@n1.R, B=Divide(A=@n1.G) {constB=255})
 ]
}
