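# Unlit helper material for Landmass landscape blueprint brushes that, going by its name and graph, computes a normalized height from a mesh's bounding box.
# Bounding-box Z is offset by a midpoint-centered, scaled Displacement texture, then split by floor and frac into two channels for high/low-byte packing.
# A separate branch builds a horizontal outward offset from local position, strongest at the bottom.
# World-Z/ObjectRadius and SceneCaptureZ-minus-RT branches stand alone; no material output connections appear.
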Material /Landmass/Landscape/BlueprintBrushes/Helpers/M_NormalizeHeight
graph {
 properties {pixelDepthOffsetMode=PDOM_Legacy, shadingModel=MSM_Unlit, usedShadingModels=Unlit}
 nodes [
  n1 = MaterialFunctionCall() {materialFunction=&BoundingBoxBased_0-1_UVW}
  n2 = TextureSampleParameter2D() {
   parameterName=Displacement
   samplerType=SAMPLERTYPE_LinearColor
   texture=/Engine/Functions/Engine_MaterialFunctions02/PivotPainter2/Black_1x1_EXR_Texture
  }
  n3 = Subtract(A=@n2.R, B=ScalarParameter() {defaultValue=0.75, parameterName=Midpoint}) {constB=0.5}
  n4 = ScalarParameter() {defaultValue=0.2, parameterName=DisplacementHeight}
  n5 = Multiply(A=Add(A=@n1.B, B=Multiply(A=@n3, B=@n4))) {constB=256}
  AppendVector(A=Divide(A=Floor(None=@n5)) {constB=256}, B=Frac(None=@n5))
  n6 = Transform(None=Normalize(VectorInput=MaterialFunctionCall() {materialFunction=&LocalPosition}.Local Position)) {transformSourceType=TRANSFORMSOURCE_Local}
  Multiply(
   A=Multiply(A=OneMinus(None=@n1.B), B=ComponentMask(None=@n6) {g=true, r=true})
   B=ScalarParameter() {defaultValue=512, parameterName=Z Expansion}
  )
  n7 = WorldPosition()
  Saturate(None=Divide(A=ComponentMask(None=@n7.XYZ) {b=true}, B=ScalarParameter() {parameterName=ObjectRadius}))
  Subtract(A=ScalarParameter() {parameterName=SceneCaptureZ}, B=TextureSampleParameter2D() {parameterName=RT, texture=null}.R)
  VertexNormalWS()
 ]
}
